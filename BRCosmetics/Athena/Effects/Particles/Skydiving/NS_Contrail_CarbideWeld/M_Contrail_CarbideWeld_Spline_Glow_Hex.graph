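# Translucent, unlit, two-sided Niagara mesh-particle material for the spline glow in the CarbideWeld skydiving contrail.
# A tinted, partly desaturated color parameter passes through HDR-from-LDR and emissive adjustment functions.
# SplineThicken takes its base and tip width from the SplineWidth dynamic parameter.
# Opacity math shapes two procedural GeneratedBand gradients, one inverted, and multiplies them by depth fade and particle alpha, without texture samples.
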Material /BRCosmetics/Athena/Effects/Particles/Skydiving/NS_Contrail_CarbideWeld/M_Contrail_CarbideWeld_Spline_Glow_Hex
graph {
 properties {
  bAutomaticallySetUsageInEditor=false
  bUsedWithNiagaraMeshParticles=true
  blendMode=BLEND_Translucent
  shadingModel=MSM_Unlit
  twoSided=true
  usedShadingModels=Unlit
 }
 nodes [
  Constant() {r=0.2}
  Constant() {r=0.3}
  n1 = Desaturation(
   None=Multiply(
    A=VectorParameter() {parameterName=ColorA, defaultValue=<rgba 2 1.1575 0.0793 1>}.RGB
    B=ScalarParameter() {defaultValue=0.65, parameterName=Param_20}
   )
   Fraction=ScalarParameter() {defaultValue=0.34, parameterName=Desaturation}
  )
  n2 = Saturate(None=@n1)
  n3 = ScalarParameter() {defaultValue=0.7, parameterName=Param_8}
  MaterialFunctionCall(Emissive=Multiply(A=MaterialFunctionCall(LDR Input=@n2, HDR Tint=@n1, HDR Intensity Multiplier=@n3) {materialFunction=&DeriveHDRfromLDR}.RemappedRange)) {materialFunction=&MF_FX_EmissiveAdjustments}
  n4 = DynamicParameter() {defaultValue=<rgba 10 1 1 1>, paramNames [SplineWidth,Param2,Param3,Param4]}
  MaterialFunctionCall(WidthBase=@n4.SplineWidth, WidthTip=@n4.SplineWidth, AngleCorrectedNormals=StaticBool()) {materialFunction=&SplineThicken}
  n5 = ScalarParameter() {defaultValue=50, parameterName=Param_10}
  n6 = MaterialFunctionCall!(
   Width=ScalarParameter() {defaultValue=0.5, parameterName=Param_3}
   Sharpness=ScalarParameter() {defaultValue=0.7, parameterName=Param_2}
   Offset
   Direction Switch=StaticBool() {value=true}
   Compare
   Input Coordinates
  ) {materialFunction=/Engine/Functions/Engine_MaterialFunctions03/Procedurals/GeneratedBand}
  n7 = ScalarParameter() {defaultValue=4}
  n8 = Saturate(None=Add(A=SmoothStep(Value=@n6.Result) {constMax=2}, B=Power(Base=@n6.Result, Exp=@n7) {constExponent=15}))
  n9 = ScalarParameter() {defaultValue=0.3, parameterName=Param_11}
  n10 = ScalarParameter() {defaultValue=-2, parameterName=Param_5}
  n11 = ScalarParameter() {defaultValue=0.1, parameterName=Param_4}
  n12 = Power(
   Base=OneMinus(None=MaterialFunctionCall(Width=@n9, Sharpness=@n10, Offset=@n11) {materialFunction=&GeneratedBand}.Result)
   Exp=ScalarParameter() {defaultValue=6, parameterName=Param_13}
  ) {constExponent=6}
  Multiply(A=Multiply(A=Multiply(A=@n8, B=@n12), B=DepthFade(FadeDistance=@n5)), B=ParticleColor().A)
  VectorParameter() {defaultValue=<rgba 1 0.5787 0.0396 1>}
 ]
}
