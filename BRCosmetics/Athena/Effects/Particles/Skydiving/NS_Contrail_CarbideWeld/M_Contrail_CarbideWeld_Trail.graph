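# Translucent, unlit, two-sided Niagara material for the CarbideWeld skydiving contrail trail.
# A panning main-shape texture, split by channel and masked along V, feeds named reroutes for the trail shape.
# Emissive blends two-color HDR gradients with panning hexagon patterns, Perlin noise and a moving band highlight.
# Dynamic parameters drive gradient intensity and a hexagon-patterned sphere-gradient erosion, with opacity scaled by depth fade and particle alpha.
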Material /BRCosmetics/Athena/Effects/Particles/Skydiving/NS_Contrail_CarbideWeld/M_Contrail_CarbideWeld_Trail
graph {
 properties {
  bAutomaticallySetUsageInEditor=false
  bUsedWithNiagaraSprites=true
  blendMode=BLEND_Translucent
  shadingModel=MSM_Unlit
  twoSided=true
  usedShadingModels=Unlit
 }
 nodes [
  n1 = TextureCoordinate()
  n2 = ScalarParameter() {defaultValue=7, parameterName=X}
  n3 = ScalarParameter() {defaultValue=10, parameterName=Y}
  n4 = Panner(
   Coordinate=Multiply(A=@n1, B=AppendVector(A=@n2, B=@n3))
   Time=NamedRerouteUsage()
   Speed=AppendVector(A=ScalarParameter() {defaultValue=0.03, parameterName=Param_12}, B=ScalarParameter() {defaultValue=2, parameterName=Param_13})
  ) {speedX=0.03, speedY=2}
  n5 = SmoothStep(Value=OneMinus(None=TextureSample(UVs=@n4) {texture=&/T_CarbideWeld_Trail_Hex_Luminance}.RGB)) {constMin=0.3}
  n6 = Saturate(None=Subtract(A=Saturate(None=@n5)) {constB=0.8})
  n7 = TextureSample(UVs=Panner(Coordinate=TextureCoordinate()) {speedX=0.1, speedY=0.25}) {samplerType=SAMPLERTYPE_Masks, texture=/BRCosmetics/Characters/Player/Male/Large/Bodies/M_LRG_Clash_V/Textures/Simple_Perlin_M}
  Clamp(None=Add(A=@n6, B=Power(Base=@n7.RGB) {constExponent=5})) {maxDefault=0.25}
  Constant3Vector() {constant=<rgba 1 0.7629 0.3978 1>}
  Constant() {r=1}
  n8 = VectorParameter() {parameterName=ColorB, defaultValue=<rgba 0.3968 0.1195 0 1>}
  n9 = VectorParameter() {parameterName=ColorA, defaultValue=<rgba 2 1.1575 0.0793 1>}
  n10 = Multiply(A=@n9.RGB, B=ScalarParameter() {defaultValue=0.85, parameterName=Param_20})
  n11 = Power(Base=ComponentMask(None=TextureCoordinate()) {g=true}) {constExponent=5}
  n12 = Desaturation(None=@n9.RGB, Fraction=Constant() {r=0.34})
  n13 = Saturate(None=@n12)
  n14 = ScalarParameter() {defaultValue=0.5, parameterName=Param_8}
  n15 = Power(Base=ComponentMask(None=@n1) {g=true})
  n16 = TextureSample(UVs=@n4) {samplerType=SAMPLERTYPE_LinearGrayscale, texture=/Game/Global/Materials/MaterialGym/Materials/TechArtPatterns/Textures/T_hexagons_DF}
  n17 = Power(Base=OneMinus(None=@n16.RGB), Exp=ScalarParameter() {defaultValue=15, parameterName=Param_7}) {constExponent=30}
  n18 = Saturate(None=Subtract(A=Multiply(A=@n15, B=@n17)) {constB=0})
  n19 = Multiply(A=@n15, B=@n6)
  n20 = LinearInterpolate(
   B=LinearInterpolate(
    A=LinearInterpolate(A=@n8.RGB, B=@n10, Alpha=@n11)
    B=MaterialFunctionCall(LDR Input=@n13, HDR Tint=@n12, HDR Intensity Multiplier=@n14) {materialFunction=&DeriveHDRfromLDR}.RemappedRange
    Alpha=Add(A=@n18, B=@n19)
   )
   Alpha=NamedRerouteUsage()
  )
  n21 = Desaturation(
   None=VectorParameter() {parameterName=ColorA, defaultValue=<rgba 2 1.1575 0.0793 1>}.RGB
   Fraction=ScalarParameter() {defaultValue=0.3, parameterName=Param_21}
  )
  n22 = MaterialFunctionCall(LDR Input=Saturate(None=@n21), HDR Tint=@n21, HDR Intensity Multiplier=ScalarParameter() {defaultValue=1.5}) {materialFunction=&DeriveHDRfromLDR}
  n23 = TextureSample(UVs=@n4) {samplerType=SAMPLERTYPE_LinearGrayscale, texture=&T_hexagons_DF}
  n24 = Power(Base=OneMinus(None=@n23.RGB), Exp=ScalarParameter() {defaultValue=3, parameterName=Param_9}) {constExponent=3}
  n25 = Power(Base=TextureSample(UVs=TextureCoordinate() {uTiling=4, vTiling=4}) {samplerType=SAMPLERTYPE_LinearGrayscale, texture=&T_SimplePerlin_Greyscale}.RGB) {constExponent=3}
  n26 = ScalarParameter() {defaultValue=0.1, parameterName=Param_1}
  n27 = Panner(
   Time=NamedRerouteUsage()
   Speed=AppendVector(A=ScalarParameter() {parameterName=Param_2}, B=ScalarParameter() {defaultValue=0.6, parameterName=Param_3})
  ) {speedY=0.3}
  n28 = Saturate(None=Multiply(
    A=Multiply(A=@n24, B=@n25)
    B=MaterialFunctionCall!(Width=@n26, Sharpness, Offset, Direction Switch, Compare, Input Coordinates=@n27) {materialFunction=/Engine/Functions/Engine_MaterialFunctions03/Procedurals/GeneratedBand}.Result
   ))
  n29 = ScalarParameter() {defaultValue=0.85, parameterName=Param_20}
  n30 = TextureCoordinate()
  n31 = LinearInterpolate(
   A=Multiply(A=VectorParameter() {parameterName=ColorB, defaultValue=<rgba 0.3968 0.1195 0 1>}.RGB, B=@n29)
   B=Multiply(A=VectorParameter() {parameterName=ColorA, defaultValue=<rgba 2 1.1575 0.0793 1>}.RGB, B=@n29)
   Alpha=Power(Base=ComponentMask(None=@n30) {g=true}) {constExponent=5}
  )
  n32 = Saturate(None=@n31)
  n33 = DynamicParameter() {defaultValue=<rgba 0 1 0 1>, paramNames [Erode?, ErodePower, Gradient, Param4]}
  MaterialFunctionCall(Emissive=LinearInterpolate(
    A=LinearInterpolate(A=@n20, B=@n22.RemappedRange, Alpha=@n28)
    B=MaterialFunctionCall(LDR Input=@n32, HDR Tint=@n31, HDR Intensity Multiplier=@n33.Gradient) {materialFunction=&DeriveHDRfromLDR}.RemappedRange
    Alpha=NamedRerouteUsage()
   )) {materialFunction=&MF_FX_EmissiveAdjustments}
  n34 = MaterialFunctionCall() {materialFunction=&ScreenResolution}
  n35 = Divide(A=ComponentMask(None=@n34.Visible Resolution) {r=true}, B=ComponentMask(None=@n34.Visible Resolution) {g=true})
  n36 = ScalarParameter() {defaultValue=22, group=Shading, parameterName=HexTiling, sliderMax=100}
  Multiply(A=ScreenPosition().ViewportUV, B=AppendVector(A=Multiply(A=@n35, B=@n36), B=@n36)) {constB=50}
  n37 = NamedRerouteUsage()
  n38 = ScalarParameter() {defaultValue=7, parameterName=X}
  n39 = ScalarParameter() {defaultValue=10, parameterName=Y}
  n40 = TextureSample(UVs=Multiply(A=TextureCoordinate(), B=AppendVector(A=@n38, B=@n39))) {samplerType=SAMPLERTYPE_LinearGrayscale, texture=&T_hexagons_DF}
  n41 = SmoothStep(Value=@n40.RGB) {constMax=0.3}
  n42 = DynamicParameter() {defaultValue=<rgba 0 1 0 1>, paramNames [Erode?, ErodePower, Gradient, Param4]}
  n43 = MaterialFunctionCall!(
   UVs=TextureCoordinate() {uTiling=2}
   CenterPosition=Constant2Vector() {g=1.5, r=1}
   Radius=LinearInterpolate(B=ScalarParameter() {defaultValue=2, parameterName=Param_5}, Alpha=@n42.ErodePower) {constB=4}
  ) {materialFunction=/Engine/Functions/Engine_MaterialFunctions01/Gradient/SphereGradient-2D}
  n44 = Multiply(A=Round(None=Multiply(A=@n37, B=@n41)), B=Step(Y=OneMinus(None=@n43.Result 0-1), X=@n42.ErodePower))
  n45 = ScalarParameter() {defaultValue=50, parameterName=DepthFade}
  Multiply(A=Multiply(A=LinearInterpolate(A=@n37, B=@n44, Alpha=@n42.Erode?), B=DepthFade(FadeDistance=@n45)), B=ParticleColor().A)
  NamedRerouteDeclaration(None=ScalarParameter() {parameterName=UVPanTime}) {name=UV Pan Time, nodeColor=<rgba 0 1 0.9062 1>}
  n46 = Multiply(A=Constant() {r=0.52}, B=LinearInterpolate(B=ParticleRandom(), Alpha=ScalarParameter() {parameterName=RandomUV?}) {constA=1})
  n47 = Multiply(A=TextureCoordinate(), B=AppendVector(A=Constant() {r=1}, B=Clamp(None=@n46) {maxDefault=0.52, minDefault=0.3}))
  n48 = VectorParameter() {group=Panner, parameterName=Panner Speed, defaultValue=<rgba 0 0.2 0 0>}
  n49 = Panner(Coordinate=@n47, Time=NamedRerouteUsage(), Speed=AppendVector(A=@n48.R, B=@n48.R)) {speedY=0.2}
  n50 = TextureObjectParameter() {
   group=Textures
   parameterName=T_MainShape
   samplerType=SAMPLERTYPE_LinearColor
   sortPriority=1
   texture=&/T_CarbideWeld_Trail
  }
  n51 = TextureSample(UVs=@n49, Tex=@n50) {samplerType=SAMPLERTYPE_LinearColor, texture=&T_CarbideWeld_Trail}
  NamedRerouteDeclaration(None=@n51.G) {name=MainShape-G, nodeColor=<rgba 0.5832 1 0 1>}
  n52 = ComponentMask(None=TextureCoordinate()) {g=true}
  n53 = SmoothStep(Max=ScalarParameter() {defaultValue=0.07, parameterName=Param_10}, Value=OneMinus(None=@n52)) {constMax=0.07}
  n54 = TextureSample(UVs=Add(A=@n47, B=Constant2Vector() {g=0.23}), Tex=@n50) {samplerType=SAMPLERTYPE_LinearColor, texture=&T_CarbideWeld_Trail}
  n55 = SmoothStep(
   Min=ScalarParameter() {defaultValue=0.7, parameterName=Param_11}
   Max=LinearInterpolate(
    A=ScalarParameter() {defaultValue=0.9, parameterName=Param_14}
    Alpha=DynamicParameter() {defaultValue=<rgba 0 1 0 1>, paramNames [Erode?, ErodePower, Gradient, Param4]}.Gradient
   ) {constB=0.95}
   Value=ComponentMask(None=TextureCoordinate()) {g=true}
  ) {constMin=0.7}
  n56 = NamedRerouteDeclaration(None=@n51.R) {name=MainShape-R, nodeColor=<rgba 1 0.0351 0 1>}
  n57 = Add(A=Multiply(A=@n54.B, B=@n55), B=Subtract(A=@n54.B, B=OneMinus(None=@n56)))
  n58 = ScalarParameter() {defaultValue=0.24, parameterName=Param_6}
  n59 = Constant() {r=1}
  n60 = Constant() {r=0.23}
  n61 = StaticBool() {value=true}
  n62 = Multiply(
   A=SmoothStep(Value=@n52) {constMax=0.7, constMin=0.2}
   B=LinearInterpolate(
    B=@n57
    Alpha=MaterialFunctionCall(Width=@n58, Sharpness=@n59, Offset=@n60, Direction Switch=@n61) {materialFunction=&GeneratedBand}.Result
   )
  )
  NamedRerouteDeclaration(None=LinearInterpolate(A=Multiply(A=@n53, B=@n57), B=Multiply(A=@n62, B=@n53), Alpha=ScalarParameter() {parameterName=UseDifferentTrail?})) {name=MainShape-B, nodeColor=<rgba 0 1 0.8435 1>}
  NamedRerouteDeclaration(None=OneMinus(None=@n55)) {name=WPOMask, nodeColor=<rgba 0 1 0.4844 1>}
  NamedRerouteDeclaration(None=@n49) {name=WPOPanner, nodeColor=<rgba 1 0.0234 0 1>}
  n63 = TextureCoordinate()
  n64 = Step(Y=OneMinus(None=ComponentMask(None=@n63) {g=true}), X=@n42.ErodePower) {constX=0.5}
  Power(Base=@n64, Exp=@n64) {constExponent=0}
  ScalarParameter() {defaultValue=0.8, parameterName=Param_4}
 ]
}
