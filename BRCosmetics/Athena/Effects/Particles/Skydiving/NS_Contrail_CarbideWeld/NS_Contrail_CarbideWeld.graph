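# Niagara skydiving contrail for the CarbideWeld cosmetic, attached to the character's skeletal mesh sockets.
# Persistent emitters draw glow sprites with lights, hand meshes, and hand and foot trail sprites aligned to the contrail direction.
# Short-lived SubUV squares spray through acceleration, drag and curl noise in matching GPU and CPU emitters.
# FN_Contrail_Setup's dive value scales spark spawning, motion and trail stretch; user colors tint sparks and glows, and time of day modulates hand opacity.
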NiagaraSystem /BRCosmetics/Athena/Effects/Particles/Skydiving/NS_Contrail_CarbideWeld/NS_Contrail_CarbideWeld
dependencies {
 usedDataInterfaces [
  &NiagaraDataInterfaceSpriteRendererInfo
  &NiagaraDataInterfaceCurve
  &NiagaraDataInterfaceArrayFloat
  &NiagaraDataInterfaceArrayInt32
  &NiagaraDataInterfaceSkeletalMesh
 ]
 usedDynamicInputs []
 usedModules [
  &self:SystemSpawnScript.SystemScriptSource.SystemScriptGraph.NiagaraNodeAssignment_0.SetVariables_#1
  &SystemState
  /Game/Effects/Niagara/Modules/Contrails/FN_Contrail_Setup
  &FN_Integrator
 ]
}
emitters {
 DigitalSquares {
  simTarget=GPUComputeSim
  emitterUpdateScript [
   &EmitterState()
   &SpawnRate(SpawnRate=&Lerp_Float(A=44, B=130, Alpha=$System.FN_Contrail_Setup.Dive Value), Use Spawn Probability=false) {module=SpawnRate002}
  ]
  particleSpawnScript [
   &PartitionParticles()
   &InitializeParticle(
    Color=&MultiplyLinearColor(
     Linear Color=&RandomRangeLinearColor(Minimum=$User.FXColorB, Maximum=$User.FXColorA)
     Scale Factor=<5 5 5 1>
    )
    Lifetime=&Lerp_Float(
     A=&RandomRangeFloat(Minimum=0.1, Maximum=0.4)
     B=&RandomRangeFloat(Minimum=0.1, Maximum=0.15)
     Alpha=$System.FN_Contrail_Setup.Dive Value
    )
    Sprite Rotation Mode=::NewEnumerator2
    Sprite Size Mode=::NewEnumerator1
    Uniform Sprite Size Max=3
    UsePositionOffset=false
    Write Material Random=true
   )
   &SkeletalMeshLocation(
    Bone / Socket Index=$Particles.PartitionParticles.Partition
    MaskSpawnGroup=false
    Mesh Sampling Type=::NewEnumerator5
    Skeletal Mesh=$System.Skeletal Mesh
    Socket Sampling Mode=::NewEnumerator5
   )
   &ShapeLocation(MaskSpawnGroup=false, Offset=<0 0 -15>, Offset Mode=::NewEnumerator0, Override Seed=false, Sphere Radius=15, UseOwnerScale=true)
   &AddVelocity(Velocity=<0 0 1>, Velocity Speed Scale=50)
  ]
  particleUpdateScript [
   &ParticleState()
   &SubUVAnimation(
    Output Alternate Renderer Binding=true
    Play Rate:NiagaraFloat=&RandomRangeFloat(Minimum=0.5, Maximum=1)
    Playback Scale Curve=DataInterface() {propertyValues {Curve {keys [
        {interpMode=RCIM_Cubic, tangentMode=RCTM_User, time=0.25, value=1}
        {arriveTangent=-1.2656, interpMode=RCIM_Cubic, leaveTangent=-1.2656, tangentMode=RCTM_User, time=1, value=0}
       ]}}}
    Random Start Frame=true
    "Sprite Renderer "=DataInterface() {propertyValues {SpriteRenderer=&self:DigitalSquares_0.NiagaraSpriteRendererProperties_1}}
    SubUV Animation Mode:ENiagara_SubUVLookupModeV2=::NewEnumerator3
    UseEndFrame=false
    UsePlayRateScale=false
    UseStartFrame=false
   )
   &SpriteRotationRate()
   &AccelerationForce(Acceleration=&Lerp_Vector(
     A=&RandomRangeVector(Minimum=<400 400 900>, Maximum=<-400 -400 1500>)
     B=&RandomRangeVector(Minimum=<700 700 5555>, Maximum=<-700 -700 5555>)
     Alpha=$System.FN_Contrail_Setup.Dive Value
    )) {module=AccelerationForce001}
   &Drag(
    Drag=&FloatFromCurve(
     FloatCurve=DataInterface() {propertyValues {Curve {
        defaultValue=1
        keys [
         {arriveTangent=2.2139, interpMode=RCIM_Cubic, leaveTangent=2.2139, tangentMode=RCTM_User, time=0, value=0}
         {interpMode=RCIM_Cubic, tangentMode=RCTM_User, time=1, value=1}
        ]
       }}}
     Scale Curve=&RandomRangeFloat(Minimum=22, Maximum=22)
    )
    Use Linear Drag=true
    Use Rotational Drag=true
   )
   &CurlNoiseForce(
    Bias Noise=false
    Noise Strength=&FloatFromCurve(
     FloatCurve=DataInterface() {propertyValues {Curve {
        defaultValue=0
        keys [
         {arriveTangent=2.6407, interpMode=RCIM_Cubic, leaveTangent=2.6407, tangentMode=RCTM_User, time=0, value=0}
         {interpMode=RCIM_Cubic, tangentMode=RCTM_User, time=1, value=1}
        ]
       }}}
     Scale Curve=2108.3284
    )
    Pan Noise=true
    Pan Noise Field=<1 0 0>
   )
   &SolveForcesAndVelocity(Clamp Velocity=false, Limit Acceleration=false)
   &ScaleSpriteSizeBySpeed(Max Scale Factor=&Vector2DFromFloat(Value=0.5), Velocity Threshold=2000)
   &ScaleSpriteSize(Uniform Curve Sprite Scale=DataInterface() {propertyValues {Curve {
       defaultValue=0
       keys [
        <at 0 0 RCIM_Cubic>
        {arriveTangent=2, interpMode=RCIM_Cubic, leaveTangent=-0.1564, tangentMode=RCTM_Break, time=0.1, value=1}
        {arriveTangent=0.2223, interpMode=RCIM_Cubic, leaveTangent=-3.2325, tangentMode=RCTM_Break, time=0.5, value=1}
        {interpMode=RCIM_Cubic, tangentMode=RCTM_User, time=1, value=0}
       ]
      }}})
   &CameraOffset(Camera Offset Amount=&RandomRangeFloat(Minimum=15, Maximum=-3))
   /SetVariables(
    Particles.GlowColor=&MultiplyLinearColor(
     Linear Color=&MultiplyLinearColor(
      Linear Color=&RandomRangeLinearColor(Minimum=$User.FXColorB, Maximum=$User.FXColorA)
      Scale Factor=<2 2 2 1>
     )
     Scale Factor=&MakeVector4(
      X=1
      Y=1
      Z=1
      W=&Multiply_Float(
       A=&RandomRangeFloat(Minimum=0.08, Maximum=0.05)
       B=&FloatFromCurve(FloatCurve=DataInterface() {propertyValues {Curve {
           defaultValue=1
           keys [{interpMode=RCIM_Cubic, tangentMode=RCTM_User, time=0.5, value=1}, {interpMode=RCIM_Cubic, tangentMode=RCTM_User, time=1, value=0}]
          }}})
      )
     )
    )
    Particles.GlowSize=&Multiply_Vector2D_ByFloat(Float=22, Vector2D=<1 1>)
   ) {isSetParameters=true, module=SetVariables_#2}
   &ScaleColor(
    Scale Alpha=&FloatFromCurve(FloatCurve=DataInterface() {propertyValues {Curve {
        defaultValue=1
        keys [{interpMode=RCIM_Cubic, tangentMode=RCTM_User, time=0.7, value=1}, {interpMode=RCIM_Cubic, tangentMode=RCTM_User, time=1, value=0}]
       }}})
    ScaleA=true
    ScaleRGB=true
   )
  ]
  rendererClasses [&NiagaraSpriteRendererProperties]
  renderers [
   {rendererClass=&NiagaraSpriteRendererProperties}
   {rendererClass=&NiagaraSpriteRendererProperties, rendererIndex=1}
  ]
 }
 DigitalSquares_CPU {
  emitterUpdateScript [
   &EmitterState()
   &SpawnRate(SpawnRate=&Lerp_Float(A=30, B=15, Alpha=$System.FN_Contrail_Setup.Dive Value), Use Spawn Probability=false) {module=SpawnRate002}
  ]
  particleSpawnScript [
   &PartitionParticles()
   &InitializeParticle(
    Color=&MultiplyLinearColor(
     Linear Color=&RandomRangeLinearColor(Minimum=$User.FXColorB, Maximum=$User.FXColorA)
     Scale Factor=<5 5 5 1>
    )
    Lifetime=&Lerp_Float(
     A=&RandomRangeFloat(Minimum=0.1, Maximum=0.4)
     B=&RandomRangeFloat(Minimum=0.1, Maximum=0.15)
     Alpha=$System.FN_Contrail_Setup.Dive Value
    )
    Sprite Rotation Mode=::NewEnumerator2
    Sprite Size Mode=::NewEnumerator1
    Uniform Sprite Size Max=3
    UsePositionOffset=false
    Write Material Random=true
   )
   &SkeletalMeshLocation(
    Bone / Socket Index=$Particles.PartitionParticles.Partition
    MaskSpawnGroup=false
    Mesh Sampling Type=::NewEnumerator5
    Skeletal Mesh=$System.Skeletal Mesh
    Socket Sampling Mode=::NewEnumerator5
   )
   &ShapeLocation(MaskSpawnGroup=false, Offset=<0 0 -15>, Offset Mode=::NewEnumerator0, Override Seed=false, Sphere Radius=15, UseOwnerScale=true)
   &AddVelocity(Velocity=<0 0 1>, Velocity Speed Scale=50)
  ]
  particleUpdateScript [
   &ParticleState()
   &SubUVAnimation(
    Output Alternate Renderer Binding=true
    Play Rate:NiagaraFloat=&RandomRangeFloat(Minimum=0.5, Maximum=1)
    Playback Scale Curve=DataInterface() {propertyValues {Curve {keys [
        {interpMode=RCIM_Cubic, tangentMode=RCTM_User, time=0.25, value=1}
        {arriveTangent=-1.2656, interpMode=RCIM_Cubic, leaveTangent=-1.2656, tangentMode=RCTM_User, time=1, value=0}
       ]}}}
    Random Start Frame=true
    "Sprite Renderer "=DataInterface() {propertyValues {SpriteRenderer=&self:DigitalSquares_CPU_0.NiagaraSpriteRendererProperties_1}}
    SubUV Animation Mode:ENiagara_SubUVLookupModeV2=::NewEnumerator3
    UseEndFrame=false
    UsePlayRateScale=false
    UseStartFrame=false
   )
   &SpriteRotationRate()
   &AccelerationForce(Acceleration=&Lerp_Vector(
     A=&RandomRangeVector(Minimum=<400 400 900>, Maximum=<-400 -400 1500>)
     B=&RandomRangeVector(Minimum=<700 700 5555>, Maximum=<-700 -700 5555>)
     Alpha=$System.FN_Contrail_Setup.Dive Value
    )) {module=AccelerationForce001}
   &Drag(
    Drag=&FloatFromCurve(
     FloatCurve=DataInterface() {propertyValues {Curve {
        defaultValue=1
        keys [
         {arriveTangent=2.2139, interpMode=RCIM_Cubic, leaveTangent=2.2139, tangentMode=RCTM_User, time=0, value=0}
         {interpMode=RCIM_Cubic, tangentMode=RCTM_User, time=1, value=1}
        ]
       }}}
     Scale Curve=&RandomRangeFloat(Minimum=22, Maximum=22)
    )
    Use Linear Drag=true
    Use Rotational Drag=true
   )
   &CurlNoiseForce(
    Bias Noise=false
    Noise Strength=&FloatFromCurve(
     FloatCurve=DataInterface() {propertyValues {Curve {
        defaultValue=0
        keys [
         {arriveTangent=2.6407, interpMode=RCIM_Cubic, leaveTangent=2.6407, tangentMode=RCTM_User, time=0, value=0}
         {interpMode=RCIM_Cubic, tangentMode=RCTM_User, time=1, value=1}
        ]
       }}}
     Scale Curve=2108.3284
    )
    Pan Noise=true
    Pan Noise Field=<1 0 0>
   )
   &SolveForcesAndVelocity(Clamp Velocity=false, Limit Acceleration=false)
   &ScaleSpriteSizeBySpeed(Max Scale Factor=&Vector2DFromFloat(Value=0.5), Velocity Threshold=2000)
   &ScaleSpriteSize(Uniform Curve Sprite Scale=DataInterface() {propertyValues {Curve {
       defaultValue=0
       keys [
        <at 0 0 RCIM_Cubic>
        {arriveTangent=2, interpMode=RCIM_Cubic, leaveTangent=-0.1564, tangentMode=RCTM_Break, time=0.1, value=1}
        {arriveTangent=0.2223, interpMode=RCIM_Cubic, leaveTangent=-3.2325, tangentMode=RCTM_Break, time=0.5, value=1}
        {interpMode=RCIM_Cubic, tangentMode=RCTM_User, time=1, value=0}
       ]
      }}})
   &CameraOffset(Camera Offset Amount=&RandomRangeFloat(Minimum=15, Maximum=-3))
   /SetVariables(
    Particles.GlowColor=&MultiplyLinearColor(
     Linear Color=&MultiplyLinearColor(
      Linear Color=&RandomRangeLinearColor(Minimum=$User.FXColorB, Maximum=$User.FXColorA)
      Scale Factor=<2 2 2 1>
     )
     Scale Factor=&MakeVector4(
      X=1
      Y=1
      Z=1
      W=&Multiply_Float(
       A=&RandomRangeFloat(Minimum=0.08, Maximum=0.05)
       B=&FloatFromCurve(FloatCurve=DataInterface() {propertyValues {Curve {
           defaultValue=1
           keys [{interpMode=RCIM_Cubic, tangentMode=RCTM_User, time=0.5, value=1}, {interpMode=RCIM_Cubic, tangentMode=RCTM_User, time=1, value=0}]
          }}})
      )
     )
    )
    Particles.GlowSize=&Multiply_Vector2D_ByFloat(Float=22, Vector2D=<1 1>)
   ) {isSetParameters=true, module=SetVariables_#2}
   &ScaleColor(
    Scale Alpha=&FloatFromCurve(FloatCurve=DataInterface() {propertyValues {Curve {
        defaultValue=1
        keys [{interpMode=RCIM_Cubic, tangentMode=RCTM_User, time=0.7, value=1}, {interpMode=RCIM_Cubic, tangentMode=RCTM_User, time=1, value=0}]
       }}})
    ScaleA=true
    ScaleRGB=true
   )
  ]
  rendererClasses [&NiagaraSpriteRendererProperties]
  renderers [
   {rendererClass=&NiagaraSpriteRendererProperties}
   {rendererClass=&NiagaraSpriteRendererProperties, rendererIndex=1}
  ]
 }
 HandGlow_Decal002 {
  emitterUpdateScript [
   &EmitterState(Life Cycle Mode=::NewEnumerator1, Loop Behavior=::NewEnumerator1, Loop Duration Mode=::NewEnumerator1, UseLoopDelay=false)
   &SpawnBurst_Instantaneous(Spawn Count=4, Use Spawn Probability=false, UseLoopCountLimit=false)
   &SpawnBurst_Instantaneous(Spawn Count=2, Spawn Group=1, Use Spawn Probability=false, UseLoopCountLimit=false) {module=SpawnBurst_Instantaneous001}
  ]
  particleSpawnScript [
   /SetVariables(Particles.VisibilityTag=$Emitter.SpawnGroup) {isSetParameters=true}
   &PartitionParticles(Number Of Partitions=4)
   &InitializeParticle(
    Color Mode=::NewEnumerator0
    Sprite Size Mode=::NewEnumerator3
    Uniform Sprite Size=&SelectFloatFromArray(Direct Array Index=$Particles.PartitionParticles.Partition, Float Selection Array=DataInterface() {propertyValues {FloatData [55,55,40,40]}})
    UsePositionOffset=false
    Write Material Random=true
   )
  ]
  particleUpdateScript [
   &ParticleState(
    Kill Particles When Lifetime Has Elapsed=false
    Let Infinitely Lived Particles Die When Emitter Deactivates=true
    Loop Particles Lifetime=true
   )
   &SkeletalMeshLocation(
    Bone / Socket Index=&SelectIntFromArray(Direct Array Index=$Particles.PartitionParticles.Partition, Int Selection Array=DataInterface() {propertyValues {IntData [0,1,2,3]}})
    MaskSpawnGroup=false
    Mesh Sampling Type=::NewEnumerator5
    Skeletal Mesh=$System.Skeletal Mesh
    Socket Sampling Mode=::NewEnumerator5
   )
   &ScaleColor(
    Scale Alpha=&FloatFromCurve(
     FloatCurve=DataInterface() {propertyValues {Curve {keys [<at 0.0001 0.3382>,<at 0.1393 0.6251>,<at 0.3202 0.5233>,<at 0.4613 0.8934>,<at 0.6241 0.3845>,<at 0.7724 0.6066>,<at 1.0006 0.3567>]}}}
     Scale Curve=&SelectFloatFromArray(
      Direct Array Index=$Particles.PartitionParticles.Partition
      Float Selection Array=DataInterface() {propertyValues {FloatData [0.5,0.5,0.1,0.1]}}
     )
    )
    Scale RGB=&MakeVectorFromLinearColorRGB(Color=$User.FXColorA)
    ScaleA=true
    ScaleRGB=true
   )
   /SetVariables(
    Particles.LightColor=&MultiplyLinearColor(Linear Color=$User.FXColorA, Scale Factor=<1 1 1 20>)
    Particles.LightPosition:NiagaraPosition=&AddVectorToPosition(Position=$StackContext.SkeletalMeshLocation.SampledPosition, Vector=<0 0 22>)
   ) {isSetParameters=true}
  ]
  rendererClasses [&NiagaraSpriteRendererProperties, &NiagaraLightRendererProperties]
  renderers [
   {rendererClass=&NiagaraSpriteRendererProperties}
   {rendererClass=&NiagaraLightRendererProperties, rendererIndex=1}
  ]
 }
 Spline_Hands {
  emitterUpdateScript [
   &EmitterState(Life Cycle Mode=::NewEnumerator1, Loop Behavior=::NewEnumerator1, Loop Duration Mode=::NewEnumerator1, UseLoopDelay=false)
   &SpawnBurst_Instantaneous(Spawn Count=2, Use Spawn Probability=false, UseLoopCountLimit=false)
  ]
  particleSpawnScript [
   &PartitionParticles()
   &InitializeParticle(Color Mode=::NewEnumerator0, Mesh Scale Mode=::NewEnumerator3, Mesh Uniform Scale=0.14, UsePositionOffset=false, Write Material Random=true)
  ]
  particleUpdateScript [
   &ParticleState(
    Kill Particles When Lifetime Has Elapsed=false
    Let Infinitely Lived Particles Die When Emitter Deactivates=true
    Loop Particles Lifetime=true
   )
   &UpdateMeshOrientation(
    Facing Interpolation=false
    Orientation Method=::NewEnumerator1
    Override Quat=false
    Side Direction:Vector3f=<1 0 1>
    Stabilization Mode:ENiagara_OrientVectorCount=::NewEnumerator3
    TransformFacing=false
    TransformSide=false
    TransformUp=false
   ) {module=UpdateMeshOrientation001}
   &SkeletalMeshLocation(
    Bone / Socket Index=$Particles.PartitionParticles.Partition
    MaskSpawnGroup=false
    Mesh Sampling Type=::NewEnumerator5
    Orientation Sampling=::NewEnumerator1
    Sampled Position Coordinate Space=::Local
    Skeletal Mesh=$System.Skeletal Mesh
    Socket Sampling Mode=::NewEnumerator5
   )
   &DynamicMaterialParameters(
    Index 0 Param 1=&FloatFromCurve(
     FloatCurve=DataInterface() {propertyValues {Curve {
        defaultValue=1
        keys [
         {interpMode=RCIM_Cubic, tangentMode=RCTM_User, time=0, value=0}
         {arriveTangent=6.6972, interpMode=RCIM_Cubic, tangentMode=RCTM_Break, time=0.3, value=0.7}
         {arriveTangent=1.7647, interpMode=RCIM_Cubic, leaveTangent=1.7647, time=0.35, value=1.5}
         <at 0.47 1 RCIM_Cubic>
        ]
       }}}
     CurveIndex=$Emitter.Age
     Scale Curve=7
    )
    Index 0 Param 3=0.2
    Module.Index0Float0Write=true
    Module.Index0Float1Write=true
    Module.Index0Float2Write=true
    Module.Index0Float3Write=true
   )
   &ScaleColor(
    Scale Alpha=&Multiply_Float(
     A=&FloatFromCurve(
      FloatCurve=DataInterface() {propertyValues {Curve {
         defaultValue=0
         keys [
          {interpMode=RCIM_Cubic, tangentMode=RCTM_User, time=0, value=0}
          {arriveTangent=4.2878, interpMode=RCIM_Cubic, tangentMode=RCTM_Break, time=0.55, value=1}
         ]
        }}}
      CurveIndex=$Emitter.Age
      Scale Curve=2
     )
     B=&FloatFromCurve(
      FloatCurve=DataInterface() {propertyValues {Curve {
         defaultValue=1
         keys [
          {interpMode=RCIM_Cubic, tangentMode=RCTM_User, time=0, value=0.8}
          {interpMode=RCIM_Cubic, tangentMode=RCTM_User, time=0.5, value=1}
          {tangentMode=RCTM_User, time=1, value=0.8}
         ]
        }}}
      CurveIndex=Linked() {linkedVariable {name=NPC.FortniteNPC.TimeOfDay_0-1, type=NiagaraFloat}}
     )
    )
    ScaleA=true
    ScaleRGB=false
   ) {module=ScaleColor001}
  ]
  rendererClasses [&NiagaraMeshRendererProperties]
  renderers [{rendererClass=&NiagaraMeshRendererProperties, rendererIndex=0}]
 }
 Trail_Feet {
  emitterUpdateScript [
   &EmitterState(Life Cycle Mode=::NewEnumerator1, Loop Behavior=::NewEnumerator1, Loop Duration Mode=::NewEnumerator1, UseLoopDelay=false)
   &SpawnBurst_Instantaneous(Spawn Count=2, Use Spawn Probability=true, UseLoopCountLimit=false)
   &SpawnBurst_Instantaneous(Spawn Count=2, Spawn Group=1, Use Spawn Probability=true, UseLoopCountLimit=false) {module=SpawnBurst_Instantaneous001}
  ]
  particleSpawnScript [
   /SetVariables(Particles.VisibilityTag=$Emitter.SpawnGroup) {isSetParameters=true}
   &PartitionParticles()
   &InitializeParticle(
    Color Mode=::NewEnumerator0
    Sprite Size=&Lerp_Vector2(
     A=&MakeVector2D(
      X=&Multiply_Float(
       A=60
       B=&Lerp_Float(A=1, B=-1, Alpha=&MakeFloatFromInt(Int=$Particles.PartitionParticles.Partition))
      )
      Y=90
     )
     B=&Vector2DFromFloat(Value=30)
     Alpha=&MakeFloatFromInt(Int=$Particles.VisibilityTag)
    )
    Sprite Size Mode=::NewEnumerator4
    UsePositionOffset=false
    Write Material Random=true
   )
  ]
  particleUpdateScript [
   &ParticleState(
    Kill Particles When Lifetime Has Elapsed=false
    Let Infinitely Lived Particles Die When Emitter Deactivates=true
    Loop Particles Lifetime=true
   )
   &ScaleSpriteSize(
    Scale Factor=&MakeVector2D(X=1, Y=&Lerp_Float(A=0.9, B=1.2, Alpha=$System.FN_Contrail_Setup.Dive Value))
    Scale Sprite Size Mode=::NewEnumerator2
   )
   &SkeletalMeshLocation(
    Bone / Socket Index=&SelectIntFromArray(Direct Array Index=$Particles.PartitionParticles.Partition, Int Selection Array=DataInterface() {propertyValues {IntData [2,3]}})
    MaskSpawnGroup=false
    Mesh Sampling Type=::NewEnumerator5
    Orientation Sampling=::NewEnumerator1
    Sampled Position Offset=&Multiply_VectorByFloat(
     Vector=&Lerp_Vector(A=<4 9 0>, B=<4 5 1>, Alpha=&MakeFloatFromInt(Int=$Particles.VisibilityTag))
     Float=&Lerp_Float(A=-1, B=1, Alpha=&MakeFloatFromInt(Int=$Particles.PartitionParticles.Partition))
    )
    Skeletal Mesh=$System.Skeletal Mesh
    Socket Sampling Mode=::NewEnumerator5
   ) {module=SkeletalMeshLocation001}
   &UpdateMeshOrientation(
    Facing Direction=$System.FN_Contrail_Setup.Direction Vector
    Facing Interpolation=false
    Orientation Method=::NewEnumerator1
    Override Quat=false
    Stabilization Mode:ENiagara_OrientVectorCount=::NewEnumerator2
    TransformFacing=false
    TransformSide=false
   )
   &AlignSpriteToMeshOrientation(
    Mesh Orientation Relative Sprite Alignment Vector=<-1 0 0>
    Mesh Orientation Relative Sprite Facing Vector:Vector3f=<0 1 0>
    SetAlignment=true
    SetFacing=true
   )
   &DynamicMaterialParameters(
    Index 0 Param 1=&FloatFromCurve(
     FloatCurve=DataInterface() {propertyValues {Curve {
        defaultValue=1
        keys [{interpMode=RCIM_Cubic, tangentMode=RCTM_User, time=0.5, value=1}, {interpMode=RCIM_Cubic, tangentMode=RCTM_User, time=1, value=0}]
       }}}
     CurveIndex=$Emitter.Age
    )
    Index 0 Param 2=&FloatFromCurve(
     FloatCurve=DataInterface() {propertyValues {Curve {
        defaultValue=0
        keys [
         {interpMode=RCIM_Cubic, tangentMode=RCTM_User, time=0, value=0}
         {arriveTangent=1, interpMode=RCIM_Cubic, leaveTangent=1, time=0.5, value=0.5}
         {arriveTangent=2.6748, interpMode=RCIM_Cubic, leaveTangent=2.6748, tangentMode=RCTM_User, time=1, value=1}
        ]
       }}}
     CurveIndex=$Emitter.Age
    )
    Index 0 Param 3=&Lerp_Float(A=0.8, B=6, Alpha=$System.FN_Contrail_Setup.Dive Value)
    Module.Index0Float0Write=true
    Module.Index0Float1Write=true
    Module.Index0Float2Write=true
    Module.Index0Float3Write=true
   )
   /SetVariables(
    Particles.GlowCameraOffset=20
    Particles.GlowColor=&MultiplyLinearColor(
     Linear Color=$User.FXColorA
     Scale Factor=&MakeVector4(
      X=1
      Y=1
      Z=1
      W=&FloatFromCurve(
       FloatCurve=DataInterface() {propertyValues {Curve {keys [<at 0.0001 0.3382>,<at 0.1393 0.6251>,<at 0.3202 0.5233>,<at 0.4613 0.8934>,<at 0.6 0.625>,<at 0.7724 0.6066>,<at 1.0006 0.3567>]}}}
       Scale Curve=0.6
      )
     )
    )
    Particles.GlowPosition:NiagaraPosition=&AddVectorToPosition(
     Position=$StackContext.Position
     Vector=&FN_FrontEndVector3Modifier(Gameplay=<0 0 0>, Front End=<0 -7 0>)
    )
   ) {isSetParameters=true}
  ]
  rendererClasses [&NiagaraSpriteRendererProperties]
  renderers [
   {rendererClass=&NiagaraSpriteRendererProperties}
   {rendererClass=&NiagaraSpriteRendererProperties, rendererIndex=1}
  ]
 }
 Trail_Hands {
  emitterUpdateScript [
   &EmitterState(Life Cycle Mode=::NewEnumerator1, Loop Behavior=::NewEnumerator1, Loop Duration Mode=::NewEnumerator1, UseLoopDelay=false)
   &SpawnBurst_Instantaneous(Spawn Count=2, Use Spawn Probability=true, UseLoopCountLimit=false)
  ]
  particleSpawnScript [
   &PartitionParticles()
   &InitializeParticle(
    Color Mode=::NewEnumerator0
    Sprite Size=&MakeVector2D(
     X=&Multiply_Float(
      A=100
      B=&Lerp_Float(A=1, B=-1, Alpha=&MakeFloatFromInt(Int=$Particles.PartitionParticles.Partition))
     )
     Y=150
    )
    Sprite Size Mode=::NewEnumerator4
    UsePositionOffset=false
    Write Material Random=true
   )
  ]
  particleUpdateScript [
   &ParticleState(
    Kill Particles When Lifetime Has Elapsed=false
    Let Infinitely Lived Particles Die When Emitter Deactivates=true
    Loop Particles Lifetime=true
   )
   &ScaleSpriteSize(
    Scale Factor=&MakeVector2D(X=1, Y=&Lerp_Float(A=1, B=1.3, Alpha=$System.FN_Contrail_Setup.Dive Value))
    Scale Sprite Size Mode=::NewEnumerator2
   )
   &SkeletalMeshLocation(
    Bone / Socket Index=$Particles.PartitionParticles.Partition
    MaskSpawnGroup=false
    Mesh Sampling Type=::NewEnumerator5
    Orientation Sampling=::NewEnumerator1
    Sampled Position Offset=&MakeVector(
     X=0
     Y=0
     Z=&Multiply_Float(
      A=0
      B=&Lerp_Float(A=-1, B=1, Alpha=&MakeFloatFromInt(Int=$Particles.PartitionParticles.Partition))
     )
    )
    Skeletal Mesh=$System.Skeletal Mesh
    Socket Sampling Mode=::NewEnumerator5
   )
   &UpdateMeshOrientation(
    Facing Direction=$System.FN_Contrail_Setup.Direction Vector
    Facing Interpolation=false
    Orientation Method=::NewEnumerator1
    Override Quat=false
    Stabilization Mode:ENiagara_OrientVectorCount=::NewEnumerator2
    TransformFacing=false
    TransformSide=false
   )
   &AlignSpriteToMeshOrientation(
    Mesh Orientation Relative Sprite Alignment Vector=<-1 0 0>
    Mesh Orientation Relative Sprite Facing Vector:Vector3f=<0 1 0>
    SetAlignment=true
    SetFacing=true
   )
   &DynamicMaterialParameters(
    Index 0 Param 1=&FloatFromCurve(
     FloatCurve=DataInterface() {propertyValues {Curve {
        defaultValue=1
        keys [
         {interpMode=RCIM_Cubic, tangentMode=RCTM_User, time=0.5, value=1}
         {arriveTangent=-6.112, interpMode=RCIM_Cubic, leaveTangent=-6.112, tangentMode=RCTM_User, time=1, value=0}
        ]
       }}}
     CurveIndex=$Emitter.Age
    )
    Index 0 Param 2=&FloatFromCurve(
     FloatCurve=DataInterface() {propertyValues {Curve {
        defaultValue=0
        keys [
         {interpMode=RCIM_Cubic, tangentMode=RCTM_User, time=0, value=0}
         {arriveTangent=1, interpMode=RCIM_Cubic, leaveTangent=1, time=0.5, value=0.5}
         {arriveTangent=0.0846, interpMode=RCIM_Cubic, leaveTangent=0.0846, tangentMode=RCTM_User, time=1, value=1}
        ]
       }}}
     CurveIndex=$Emitter.Age
    )
    Index 0 Param 3=&Lerp_Float(A=0.8, B=4, Alpha=$System.FN_Contrail_Setup.Dive Value)
    Module.Index0Float0Write=true
    Module.Index0Float1Write=true
    Module.Index0Float2Write=true
    Module.Index0Float3Write=true
   ) {module=DynamicMaterialParameters001}
   &CameraOffset(Camera Offset Amount=3)
   &ScaleColor(
    Scale Alpha=&FloatFromCurve(
     FloatCurve=DataInterface() {propertyValues {Curve {
        defaultValue=1
        keys [
         {interpMode=RCIM_Cubic, tangentMode=RCTM_User, time=0, value=0.8}
         {interpMode=RCIM_Cubic, tangentMode=RCTM_User, time=0.5, value=1}
         {tangentMode=RCTM_User, time=1, value=0.8}
        ]
       }}}
     CurveIndex=Linked() {linkedVariable {name=NPC.FortniteNPC.TimeOfDay_0-1, type=NiagaraFloat}}
    )
    ScaleA=true
    ScaleRGB=false
   )
  ]
  rendererClasses [&NiagaraSpriteRendererProperties]
  renderers [{rendererClass=&NiagaraSpriteRendererProperties}]
 }
}
summary {
 systemName=NS_Contrail_CarbideWeld
 emitters [
  {emitterName=DigitalSquares, simTarget=GPUComputeSim, rendererClasses [&NiagaraSpriteRendererProperties]}
  {
   emitterName=HandGlow_Decal002
   rendererClasses [&NiagaraSpriteRendererProperties, &NiagaraLightRendererProperties]
  }
  {emitterName=Spline_Hands, rendererClasses [&NiagaraMeshRendererProperties]}
  {emitterName=Trail_Hands, rendererClasses [&NiagaraSpriteRendererProperties]}
  {emitterName=Trail_Feet, rendererClasses [&NiagaraSpriteRendererProperties]}
  {emitterName=DigitalSquares_CPU, rendererClasses [&NiagaraSpriteRendererProperties]}
 ]
 userVariables [
  User.bIsFrontEnd {type=NiagaraBool}
  User.bIsFrontEndPreview {type=NiagaraBool}
  User.FXColorA {type=LinearColor, description="", defaultValue {struct=LinearColor, value=<rgba 1 0.8557 0.1048 1>}}
  User.FXColorB {type=LinearColor, description="", defaultValue {struct=LinearColor, value=<rgba 0.3968 0.1195 0 1>}}
  User.ParaGlideLeanAlpha {type=NiagaraFloat, defaultValue=0, description=""}
  User.RibbonIDOffset {type=NiagaraInt32}
 ]
}
